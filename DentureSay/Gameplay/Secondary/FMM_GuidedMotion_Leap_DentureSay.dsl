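; Fort movement mode subclass for the DentureSay secondary ability's guided-motion leap, adjusting the strength of its constant force.
; It overrides GetStrength_ConstantForce to return zero while the owner's movement component is moving on the ground.
; When airborne, or when no movement component is available, it defers to the parent class's strength calculation.
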(fn GetStrength_ConstantForce
  (params (GuidedMotionRuntimeData :type Fort%20Movement%20Mode%20Guided%20Motion%20Runtime%20Data%20Object%20Reference))
  (bind _ownermovementcomponent (Variables|FortMovementMode_BaseExtLogic|GetOwnerMovementComponent)
    (:then
      (if (AI|Components|NavMovement|IsMovingonGround _ownermovementcomponent)
        (return 0.0)
        (else
          _)))
    (:else
      _))
  (bind _returnvalue (|CallParentFunction GuidedMotionRuntimeData))
  (return _returnvalue))
