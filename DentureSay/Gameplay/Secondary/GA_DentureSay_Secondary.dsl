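; DentureSay's secondary gameplay ability, a hold-and-release cast that spawns strikes, with a leap and hover available while falling.
; Activation runs the parent logic, plays the fire montage, then plays the cast animation on input release.
; Montage notify tags drive the rest: the leap tag leaps if falling, and the cast tag commits, applies the spawn-strikes GameplayEffect to the owner and ends.
; The SetHover override permits parent hover only while falling.
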(event Ability|EventActivateAbility
  (|CallParentFunction)
  (Ability|Tasks|PlayMontageAndWait "None" "/DentureSay/Assets/Proto/DentureSay_Secondary_Fire_M.DentureSay_Secondary_Fire_M")
  (Ability|Tasks|WaitInputRelease false
    (:OnRelease
      (CallFunction|PlayCastAnimation))))

(event EventCastMontageBlendOut)

(event EventCastMontageNotifyStart
  (params (NotifyTag :type Gameplay%20Tag%20Structure))
  ;[0.02755 0.15 0.08681 0.5] Start strikes on montage notify. Notify on client RPCs to server with current targetng transform
  (if (GameplayTags|Equal(GameplayTag) NotifyTag (Variables|PowerLeap|GetTag_LeapNotify))
    (if (CallFunction|IsPlayerFalling)
      (CallFunction|Leap))
    (elif (== NotifyTag (Variables|Default|GetTagCastNotify))
      (Utilities|PowerCommitAbility (Variables|AbilityCommit|GetIgnoreAbilityCommit)
        (:Pass
          _)
        (:Fail
          (CallFunction|PowerCommitFailed))
        (:Ignored
          _))
      (Ability|ApplyGameplayEffectToOwner (Variables|Default|GetGE_SpawnStrikes))
      (Ability|EndAbility))))

(fn IsPlayerFalling
  (bind _playerpawn (Variables|Default|GetPlayerPawn)
    (:then
      (bind _charactermovement (Variables|/Script/Engine.Character|GetCharacterMovement _playerpawn)
        (:then
          (return (AI|Components|NavMovement|IsFalling _charactermovement)))
        (:else
          _)))
    (:else
      _))
  (return false))

(fn SetHover
  (params (On :type Boolean))
  (|CallParentFunction (and On (CallFunction|IsPlayerFalling))))
