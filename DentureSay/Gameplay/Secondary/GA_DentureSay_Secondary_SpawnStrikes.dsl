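; DentureSay secondary ability that fires a timed line of ground strikes along the player's pitch-clamped camera aim.
; Each timer tick offsets the strike by XY spacing, then async targeting finds non-damageable terrain, favouring ground below the player and measuring a ceiling clamp.
; A second async request applies the damage effect and records hit actors to ignore.
; A gameplay cue packs location, height and water tag for FX; the last strike ends the ability.
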(event Ability|EventActivateAbility
  (|CallParentFunction)
  (bind _output_get (Variables|Runtime|SetPlayerPawn (CallFunction|GetAvatarasFortPlayerPawn)))
  (bind _returnvalue (TargetingSystem|GetPlayerCameraTargetingTransform _output_get))
  (CallFunction|StartStrikes (CallFunction|ClampTargetingPitch _returnvalue)))

(event Custom|StartStrikes
  (params (TargetingTransform :type Transform))
  (Variables|Runtime|SetTargetingTransform TargetingTransform)
  (Variables|Runtime|SetCurrentStrikeIndex 0)
  (Utilities|Array|Clear (Variables|Runtime|GetActorsToIgnore))
  (bind _returnvalue (Utilities|Time|SetTimerbyEvent (EventDispatchers|CreateEvent :!event Custom|SpawnStrike) (ScalableFloat|GetValueatLevel (Variables|Hotfix|GetRow_StrikeInterval)) true))
  (Variables|Runtime|SetStrikeTimerHandle _returnvalue))

(event Custom|SpawnStrike
  (bind _targetingtransform_location (Math|Transform|BreakTransform (Variables|GetTargetingTransform_Location)))
  ;[0.05835 0.15 0.0591 0.5] Spawn Strike - Update strike location and index, then search for ground
  (bind _output_get (Variables|Runtime|SetCurrentStrikeHitTarget (Collision|MakeHitResult false false 0.0 0.0 "0, 0, 0" (Default|GetStrikeOrigin) "0.000000,0.000000,0.000000")))
  (Math|Integer|IncrementInt (Variables|Runtime|GetCurrentStrikeIndex)
    (:%20%20
      (bind (_returnvalue_x _returnvalue_y _returnvalue_z) (Math|Vector|BreakVector (Utilities|Operators|Add (.ImpactPoint (Collision|BreakHitResult _output_get)) (Math|Vector|MakeVector 0.0 0.0 (ScalableFloat|GetValueatLevel (Variables|Hotfix|GetRow_StrikeTraceLength))))))
      (TargetingSystem|AsyncRequest|StartAsyncTargetingRequest (GameInstanceSubsystems|GetTargetingSubsystem) (Variables|Settings|GetStrikeLocationTargetingPreset) (Utilities|Struct|MakeTargetingSourceContext 0 (Variables|Runtime|GetPlayerPawn) (Math|Vector|MakeVector _returnvalue_x _returnvalue_y (Math|Float|Max(Float) _returnvalue_z (.z _targetingtransform_location))) "None" self) (delegate Custom|On%20Strike%20Location%20Targeting%20Complete)))))

(event Custom|On%20Strike%20Location%20Targeting%20Complete
  (params (TargetingRequestHandle :type Targeting%20Request%20Handle%20Structure))
  ;[0.05729 0.15 0.1462 0.5] Once ground is found, start damage targeting on server and FX on client
  (Default|CalculateStrikeLocation (TargetingSystem|TargetingResults|GetTargetingResults (GameInstanceSubsystems|GetTargetingSubsystem) TargetingRequestHandle))
  (CallFunction|PerformStrikeTargeting))

(event Custom|Perform%20Strike%20Targeting
  (bind (_bblockinghit _binitialoverlap _time _distance _location _impactpoint) (Collision|BreakHitResult (Variables|Runtime|GetCurrentStrikeHitTarget)))
  ;[0.15 0.04199 0.04342 0.5] Strike Damage
  (TargetingSystem|AsyncRequest|StartAsyncTargetingRequest (GameInstanceSubsystems|GetTargetingSubsystem) (Variables|Settings|GetStrikeDamageTargetingPreset) (Utilities|Struct|MakeTargetingSourceContext 0 (Variables|Runtime|GetPlayerPawn) _impactpoint "None" self) (delegate Custom|On%20Strike%20Damage%20Targeting%20Complete)))

(event Custom|On%20Strike%20Damage%20Targeting%20Complete
  (params (TargetingRequestHandle :type Targeting%20Request%20Handle%20Structure))
  (bind _returnvalue (GameInstanceSubsystems|GetTargetingSubsystem))
  ;[0.15 0.04199 0.04342 0.5] Strike Damage
  (Utilities|FlowControl|Sequence
    (:then_0
      (bind _asfort_ability_system_component_athena (Utilities|Casting|CastToFortAbilitySystemComponentAthena (Ability|GetAbilitySystemComponentfromActorInfo)))
      (Ability|Tasks|ApplyGameplayEffectToTargetingHandleAsync _asfort_ability_system_component_athena TargetingRequestHandle "/DentureSay/Gameplay/Secondary/GE_DentureSay_Secondary_Damage.GE_DentureSay_Secondary_Damage_C" 0 1 (GameplayEffects|MakeFortEffectContext _asfort_ability_system_component_athena self (Variables|Runtime|GetPlayerPawn) (Ability|GetCurrentSourceWeapon) 0) (Variables|Hotfix|GetRow_MaxTargetsToProcessPerTick)))
    (:then_1
      (Utilities|IsValid _returnvalue
        (:Is%20Valid
          (for _array_element (TargetingSystem|TargetingResults|GetTargetingResults _returnvalue TargetingRequestHandle)
            (Utilities|Array|AddUnique (Variables|Runtime|GetActorsToIgnore) (.HitActor (Collision|BreakHitResult _array_element)))
            (Development|PrintString (Utilities|GetDisplayName (.HitActor (Collision|BreakHitResult _array_element))))))))
    (:then_2
      (CallFunction|OnStrike))))

(event Custom|OnStrike
  (bind (_bblockinghit _binitialoverlap _time _distance _location _impactpoint _normal _impactnormal _physmat _hitactor) (Collision|BreakHitResult (Variables|Runtime|GetCurrentStrikeHitTarget)))
  ;[0.0748 0.05797 0.15 0.5] Handle strike
  ;;[0.1451 0.04199 0.15 0.5] Strike GC - Doing some fancy stuff to replicate strike details properly
  ;; Location = Bottom, Magnitude = Top distance from bottom, Normal = Whether ground was hit (0 if not)
  (Ability|ExecuteGameplayCueWithParamsOnOwner (Variables|Settings|GetStrikeGameplayCue) (Ability|GameplayCue|MakeGameplayCueParameters :NormalizedMagnitude 0.0 :RawMagnitude (Math|Float|SelectFloat (ScalableFloat|GetValueatLevel (Variables|Hotfix|GetRow_StrikeTraceLength)) (Math|Float|Min(Float) (ScalableFloat|GetValueatLevel (Variables|Hotfix|GetRow_StrikeTraceLength)) (Variables|Runtime|GetCeilingClampDistance)) (< (Variables|Runtime|GetCeilingClampDistance) 0)) :AggregatedTargetTags (select (Utilities|IsA _hitactor "/Script/FortniteGame.FortWaterBodyActor") "(GameplayTags=((TagName=\"Fort.SurfaceType.Water\")))" 0) :Location _impactpoint :Normal _normal :GameplayEffectLevel 1 :AbilityLevel 1 :bReplicateLocationWhenUsingMinimalRepProxy false))
  ;;[0.15 0.02834 0.01428 0.5] End Ability on last strike
  (if (>= (Variables|Runtime|GetCurrentStrikeIndex) (- (ScalableFloat|GetValueatLevel (Variables|Hotfix|GetRow_NumStrikes)) 1))
    (Utilities|Time|ClearandInvalidateTimerbyHandle (Variables|Runtime|GetStrikeTimerHandle))
    (Ability|EndAbility)))

(fn CalculateStrikeLocation
  (params (TerrainHits :type Array%20of%20Hit%20Result%20Structures))
  (bind _lowindex (LocalVariables|GetLowIndex))
  (bind _validterrainhits (LocalVariables|GetValidTerrainHits))
  (Variables|Runtime|SetCeilingClampDistance -1.0)
  (for _array_element TerrainHits
    ;[0.15 0.1335 0.04341 0.5] Find non-damagable hits
    (Utilities|IsValid (.HitActor (Collision|BreakHitResult _array_element))
      (:Is%20Valid
        (if (not (Variables|/Script/Engine.Actor|GetbCanBeDamaged (.HitActor (Collision|BreakHitResult _array_element))))
          (Utilities|Array|Add (LocalVariables|GetValidTerrainHits) _array_element)))))
  ;[0.15 0.04482 0.04196 0.5] No locations found. Early out
  (if (not (Utilities|Array|IsEmpty (LocalVariables|GetValidTerrainHits)))
    (bind _playerpawn (Variables|Runtime|GetPlayerPawn)
      (:then
        (bind (_returnvalue_x _returnvalue_y _returnvalue_z) (Math|Vector|BreakVector (Transformation|GetActorLocation _playerpawn)))
        (LocalVariables|SetPlayerHeight _returnvalue_z)
        (bind (_arrayindex _arrayelement) (Utilities|Array|ReverseforEachLoop _validterrainhits)
          (:LoopBody
            ;[0.03863 0.15 0.1026 0.5] Low index = last terrain location or highest location under player. CeilingClampDistance = first location above player (excluding last terrain location)
            ; Assumes locations array is in height decending order
            (if (== _arrayindex (- (Utilities|Array|Length _validterrainhits) 1))
              _
              (elif (< (.z (.ImpactPoint (Collision|BreakHitResult _arrayelement))) (LocalVariables|GetPlayerHeight))
                (if (< (LocalVariables|GetLowHeight) (.z (.ImpactPoint (Collision|BreakHitResult _arrayelement))))
                  _)
                (else
                  (if (< (Variables|Runtime|GetCeilingClampDistance) 0)
                    (Variables|Runtime|SetCeilingClampDistance (Utilities|Operators|Subtract (.z (.ImpactPoint (Collision|BreakHitResult _arrayelement))) (LocalVariables|GetLowHeight) (Variables|Hotfix|GetRow_StrikeDamageRadius))) ; Distance between found low point and \nfirst non-low point above player
                  ))))
            (LocalVariables|SetLowIndex _arrayindex)
            (LocalVariables|SetLowHeight (.z (.ImpactPoint (Collision|BreakHitResult _arrayelement)))))
          (:Completed
            ;[0.05596 0.15 0.03799 0.5] Save found strike origin
            (if (>= _lowindex 0)
              (Variables|Runtime|SetCurrentStrikeHitTarget (Utilities|Array|Get(aref) (LocalVariables|GetValidTerrainHits) _lowindex))
              (else
                (Utilities|Debugging|RaiseScriptError "DentureSay Secondary: No low location found"))))))
      (:else))))

(fn ClampTargetingPitch
  (params (InTransform :type Transform))
  (bind (_location _rotation _scale) (.location InTransform))
  (bind (_rotation_roll _rotation_pitch _rotation_yaw) (Math|Rotator|BreakRotator _rotation))
  (return (Math|Transform|MakeTransform _location (Math|Rotator|MakeRotator _rotation_roll (Math|Float|ClampAngle _rotation_pitch (* (ScalableFloat|GetValueatLevel (Variables|Hotfix|GetRow_MaxTargetingPitch)) -1.0) (ScalableFloat|GetValueatLevel (Variables|Hotfix|GetRow_MaxTargetingPitch))) _rotation_yaw) _scale)))

(fn GetStrikeOrigin
  ;[0.15 0.1435 0.03911 0.5] Get location of each strike based on index * spacing. Add extra distance for first strike and add the strike down distance to get the bottom location. All spacing is XY based.
  (return (Utilities|Operators|Add (.location (Variables|Runtime|GetTargetingTransform)) (* (Math|Vector|GetForwardVector (.rotation (Variables|Runtime|GetTargetingTransform))) (/ (ScalableFloat|GetValueatLevel (Variables|Hotfix|GetRow_FirstStrikeDistance)) (Math|Vector|VectorLengthXY (Math|Vector|GetForwardVector (.rotation (Variables|Runtime|GetTargetingTransform)))))) (Utilities|Operators|Multiply (Math|Vector|GetForwardVector (.rotation (Variables|Runtime|GetTargetingTransform))) (/ (ScalableFloat|GetValueatLevel (Variables|Hotfix|GetRow_Spacing)) (Math|Vector|VectorLengthXY (Math|Vector|GetForwardVector (.rotation (Variables|Runtime|GetTargetingTransform))))) (Math|Conversions|ToFloat(Integer) (Variables|Runtime|GetCurrentStrikeIndex))) (Math|Vector|MakeVector 0.0 0.0 (* (ScalableFloat|GetValueatLevel (Variables|Hotfix|GetRow_StrikeDownDistance)) -1.0)))))
