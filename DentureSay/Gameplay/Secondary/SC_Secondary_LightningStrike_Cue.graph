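# Sound cue for a secondary lightning strike impact, built from Zeus lightning samples with per-distance attenuation.
# Branches on islocal and isenemy play one shared mix, halving volume only for non-local, non-enemy sources.
# Close randomized impacts run beside a quieter distant-class layer of random medium plus all three distant waves together, each layer pitch-varied.
# Concurrency override caps it at two instances, stopping the oldest.
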SoundCue /DentureSay/Gameplay/Secondary/SC_Secondary_LightningStrike_Cue
root=Branch_0
nodes [
 {
  id=Attenuation_118
  children [Random_4]
  properties {
   attenuationSettings=/SunRoseZeusGameplay/Sounds/Ability/Sunrose_Zeus_Lightning_Impt_Close_Att
   attenuationOverrides {stereoSpread=29871.3984}
  }
 }
 {
  id=Attenuation_119
  children [Random_1]
  properties {attenuationSettings=/SunRoseZeusGameplay/Sounds/Ability/Sunrose_Zeus_Lightning_Impt_Med_Att}
 }
 {
  id=Attenuation_120
  children [Mixer_1]
  properties {attenuationSettings=/SunRoseZeusGameplay/Sounds/Ability/Sunrose_Zeus_Lightning_Impt_Dist_Att}
 }
 {id=Branch_0, children [Modulator_2,Branch_3,Modulator_2], properties {boolParameterName=islocal}}
 {id=Branch_3, children [Modulator_4,Modulator_3,Modulator_4], properties {boolParameterName=isenemy}}
 {id=Mixer_0, children [Modulator_0,Modulator_1], properties {inputVolume [1,0.7]}}
 {id=Mixer_1, children [WavePlayer_0,WavePlayer_1,WavePlayer_2], properties {inputVolume [1,1,1]}}
 {id=Mixer_2, children [Attenuation_119, Attenuation_120], properties {inputVolume [1,1]}}
 {id=Modulator_0, children [SoundClass_77], properties {pitchMax=1.1, pitchMin=0.9}}
 {id=Modulator_1, children [SoundClass_76], properties {pitchMax=1.2, pitchMin=0.85}}
 {id=Modulator_2, children [Mixer_0], properties {pitchMax=1, pitchMin=1, volumeMax=1, volumeMin=1}}
 {id=Modulator_3, children [Mixer_0], properties {pitchMax=1, pitchMin=1, volumeMax=0.5, volumeMin=0.5}}
 {id=Modulator_4, children [Mixer_0], properties {pitchMax=1, pitchMin=1, volumeMax=1, volumeMin=1}}
 {id=Random_1, children [WavePlayer_5,WavePlayer_6,WavePlayer_7], properties {weights [1,1,1]}}
 {id=Random_4, children [WavePlayer_3,WavePlayer_4], properties {weights [1,1]}}
 {id=SoundClass_76, children [Mixer_2], properties {soundClassOverride=&Fort_WeaponRanged_Fire_Distant}}
 {
  id=SoundClass_77
  children [Attenuation_118]
  properties {soundClassOverride=&Fort_WeaponRanged_Fire_3P_PSM}
 }
 {id=WavePlayer_0, properties {soundWaveAssetPtr=/SunRoseZeusGameplay/Sounds/Ability/Wavs/Sunrose_ZeusLightning_Impact_Dist_01}}
 {id=WavePlayer_1, properties {soundWaveAssetPtr=/SunRoseZeusGameplay/Sounds/Ability/Wavs/Sunrose_ZeusLightning_Impact_Dist_02}}
 {id=WavePlayer_2, properties {soundWaveAssetPtr=/SunRoseZeusGameplay/Sounds/Ability/Wavs/Sunrose_ZeusLightning_Impact_Dist_03}}
 {id=WavePlayer_3, properties {soundWaveAssetPtr=/SunRoseZeusGameplay/Sounds/Ability/Wavs/Sunrose_ZeusLightning_Impact_Close_01}}
 {id=WavePlayer_4, properties {soundWaveAssetPtr=/SunRoseZeusGameplay/Sounds/Ability/Wavs/Sunrose_ZeusLightning_Impact_Close_02}}
 {id=WavePlayer_5, properties {soundWaveAssetPtr=/SunRoseZeusGameplay/Sounds/Ability/Wavs/Sunrose_ZeusLightning_Impact_Med_01}}
 {id=WavePlayer_6, properties {soundWaveAssetPtr=/SunRoseZeusGameplay/Sounds/Ability/Wavs/Sunrose_ZeusLightning_Impact_Med_02}}
 {id=WavePlayer_7, properties {soundWaveAssetPtr=/SunRoseZeusGameplay/Sounds/Ability/Wavs/Sunrose_ZeusLightning_Impact_Med_03}}
]
properties {
 bOverrideConcurrency=true
 maxDistance=30000
 soundClassObject=&Fort_WeaponRanged
 volumeMultiplier=0.7
 concurrencyOverrides {maxCount=2, resolutionRule=StopOldest, retriggerTime=0.1}
 concurrencySet [/SunRoseZeusGameplay/Sounds/Ability/Sunrose_ZeusLightningBolts_Conc]
}
